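# Sound cue for the guided missile's activation sound, routed to a third-person ranged weapon fire sound class.
# It picks one of three equally weighted Unvault activation waves at random and passes it through a modulator left at default settings.
# It references a missile launch attenuation asset but overrides it with its own long-range falloff, which adds distance-based filtering.
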SoundCue /Game/Athena/Sounds/Weapons/Guided_Missile/GuidedMissile_Activate_Cue
root=Modulator_0
nodes [
 {id=Modulator_0, children [Random_0]}
 {id=Random_0, children [WavePlayer_1,WavePlayer_2,WavePlayer_3], properties {bShouldExcludeFromBranchCulling=true, weights [1,1,1]}}
 {id=WavePlayer_1, properties {soundWaveAssetPtr=/Game/Athena/Sounds/Weapons/Guided_Missile/Unvault/GuidedMissile_Unvault_Activate_01}}
 {id=WavePlayer_2, properties {soundWaveAssetPtr=/Game/Athena/Sounds/Weapons/Guided_Missile/Unvault/GuidedMissile_Unvault_Activate_02}}
 {id=WavePlayer_3, properties {soundWaveAssetPtr=/Game/Athena/Sounds/Weapons/Guided_Missile/Unvault/GuidedMissile_Unvault_Activate_03}}
]
properties {
 attenuationSettings=/Game/Sounds/Attenuation/Fort_Weapon_Projectile_Launch_Missile_Athena_Att
 bOverrideAttenuation=true
 maxDistance=13545.416
 soundClassObject=&Fort_WeaponRanged_Fire_3P_PSM
 volumeMultiplier=1.2
 attenuationOverrides {
  bAttenuateWithLPF=true
  bEnableLogFrequencyScaling=true
  falloffDistance=13139.8223
  hPFFrequencyAtMax=175.0834
  lPFFrequencyAtMax=771.233
  lPFRadiusMax=9108.7695
  attenuationShapeExtents {x=405.5941}
 }
}
