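# Sound cue for the guided missile's looping in-air sound, which responds to banking.
# The BankingSound parameter crossfades between a banking loop, weighted to higher values, and a jet loop, weighted to lower values.
# A mixer layers that crossfade with a constant distortion loop at slightly reduced volume.
# All three loops share close-missile projectile attenuation and a third-person ranged-weapon sound class, with concurrency limiting playback to one per owner.
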SoundCue /Game/Athena/Sounds/Weapons/Guided_Missile/GuidedMissile_InAir_Banking_Loop_Cue
root=Mixer_0
nodes [
 {
  id=Attenuation_1
  children [WavePlayer_0]
  properties {attenuationSettings=/Game/Sounds/Attenuation/Fort_Weapon_Projectile_Close_Missile_Athena_Att}
 }
 {
  id=Attenuation_2
  children [WavePlayer_3]
  properties {attenuationSettings=&Fort_Weapon_Projectile_Close_Missile_Athena_Att}
 }
 {
  id=Attenuation_7
  children [WavePlayer_8]
  properties {attenuationSettings=&Fort_Weapon_Projectile_Close_Missile_Athena_Att}
 }
 {id=Mixer_0, children [ParamCrossFade_1, SoundClass_8], properties {inputVolume [1,0.8]}}
 {
  id=ParamCrossFade_1
  children [SoundClass_3,SoundClass_1]
  properties {
   paramName=BankingSound
   crossFadeInput [
    {fadeInDistanceEnd=1, fadeInDistanceStart=0.3, fadeOutDistanceEnd=0.4, fadeOutDistanceStart=0.9, volume=1.4888}
    {fadeInDistanceEnd=0, fadeInDistanceStart=0, fadeOutDistanceEnd=1.2, fadeOutDistanceStart=0.7, volume=0.2735}
   ]
  }
 }
 {
  id=SoundClass_1
  children [Attenuation_1]
  properties {soundClassOverride=&Fort_WeaponRanged_Fire_3P_PSM}
 }
 {
  id=SoundClass_3
  children [Attenuation_2]
  properties {soundClassOverride=&Fort_WeaponRanged_Fire_3P_PSM}
 }
 {
  id=SoundClass_8
  children [Attenuation_7]
  properties {soundClassOverride=&Fort_WeaponRanged_Fire_3P_PSM}
 }
 {
  id=WavePlayer_0
  properties {bLooping=true, soundWaveAssetPtr=/Game/Athena/Sounds/Weapons/Guided_Missile/Unvault/GuidedMissile_Unvault_InAir_Loop_Jet_01}
 }
 {
  id=WavePlayer_3
  properties {bLooping=true, soundWaveAssetPtr=&/GuidedMissile_InAir_Banking_Loop_04}
 }
 {
  id=WavePlayer_8
  properties {bLooping=true, soundWaveAssetPtr=/Game/Athena/Sounds/Weapons/Guided_Missile/Unvault/GuidedMissile_Unvault_InAir_Loop_Distort_01}
 }
]
properties {
 maxDistance=4500
 soundClassObject=&Fort_WeaponRanged_Fire_3P_PSM
 volumeMultiplier=0.4
 concurrencyOverrides {bLimitToOwner=true, maxCount=1, resolutionRule=PreventNew}
 concurrencySet [/Game/Sounds/Concurrency_Rules/OneMaximum_Owner_StopOldest]
}
