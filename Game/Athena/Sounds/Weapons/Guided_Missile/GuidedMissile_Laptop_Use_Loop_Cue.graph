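# Sound cue for the looping sound of the Guided Missile weapon's laptop in use, routed to the Fort_Weapons sound class.
# A quality-level node picks between two looping wave players: the standard wave for the first two levels and an "_opt" variant for the last two.
# The cue plays quietly at slightly lowered pitch, with its own short-range logarithmic attenuation override.
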SoundCue /Game/Athena/Sounds/Weapons/Guided_Missile/GuidedMissile_Laptop_Use_Loop_Cue
root=QualityLevel_0
nodes [
 {id=QualityLevel_0, children [WavePlayer_1,WavePlayer_1,WavePlayer_0,WavePlayer_0]}
 {
  id=WavePlayer_0
  properties {bLooping=true, soundWaveAssetPtr=&/GuidedMissile_Laptop_Use_Loop_02_opt}
 }
 {id=WavePlayer_1, properties {bLooping=true, soundWaveAssetPtr=&/GuidedMissile_Laptop_Use_Loop_02}}
]
properties {
 bOverrideAttenuation=true
 maxDistance=1000
 pitchMultiplier=0.9
 soundClassObject=/Game/Sounds/FortSoundClassesAndModes/Fort_Weapons
 volumeMultiplier=0.163
 attenuationOverrides {distanceAlgorithm=Logarithmic, falloffDistance=900, attenuationShapeExtents {x=100}}
}
