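# First-person launch sound cue for the Guided Missile weapon.
# A Random node picks one of three equally weighted Unvault launch waves and feeds it through a Modulator with pitch fixed at normal.
# It uses first-person projectile attenuation and the first-person ranged-fire sound class.
# Playback is governed by the shared first-person gunfire concurrency plus a missile-specific Unvault fire concurrency.
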SoundCue /Game/Athena/Sounds/Weapons/Guided_Missile/GuidedMissile_Launch_1P_Cue
root=Modulator_6
nodes [
 {id=Modulator_6, children [Random_0], properties {pitchMax=1, pitchMin=1}}
 {id=Random_0, children [WavePlayer_1,WavePlayer_2,WavePlayer_3], properties {bShouldExcludeFromBranchCulling=true, weights [1,1,1]}}
 {id=WavePlayer_1, properties {soundWaveAssetPtr=/Game/Athena/Sounds/Weapons/Guided_Missile/Unvault/GuidedMissile_Unvault_Launch_01}}
 {id=WavePlayer_2, properties {soundWaveAssetPtr=/Game/Athena/Sounds/Weapons/Guided_Missile/Unvault/GuidedMissile_Unvault_Launch_02}}
 {id=WavePlayer_3, properties {soundWaveAssetPtr=/Game/Athena/Sounds/Weapons/Guided_Missile/Unvault/GuidedMissile_Unvault_Launch_03}}
]
properties {
 attenuationSettings=&Fort_Weapon_Projectile_FirstPerson_Athena_Att
 maxDistance=2097152
 soundClassObject=&Fort_WeaponRanged_Fire_1P_PSM
 volumeMultiplier=1
 concurrencyOverrides {bLimitToOwner=true, maxCount=1, resolutionRule=PreventNew}
 concurrencySet [
  &Weapon_Gunfire_FirstPerson_Conc
  /Game/Athena/Sounds/Weapons/Guided_Missile/Unvault/GuidedMissile_Unvault_Fire
 ]
}
