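# Sound cue for the Guided Missile weapon's launch in Athena, built on the Unvault launch recordings.
# An equally weighted Random node picks one of three waves, feeding a Modulator that holds pitch fixed.
# Close-projectile attenuation and a third-person ranged-fire sound class wrap the result.
# The Unvault Fire concurrency asset governs overlapping plays, and the Random node is excluded from branch culling.
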SoundCue /Game/Athena/Sounds/Weapons/Guided_Missile/GuidedMissile_Launch_Cue
root=SoundClass_1
nodes [
 {id=Attenuation_9, children [Modulator_6], properties {attenuationSettings=&Fort_Weapon_Projectile_Close_Athena_Att}}
 {id=Modulator_6, children [Random_0], properties {pitchMax=1, pitchMin=1}}
 {id=Random_0, children [WavePlayer_1,WavePlayer_4,WavePlayer_5], properties {bShouldExcludeFromBranchCulling=true, weights [1,1,1]}}
 {
  id=SoundClass_1
  children [Attenuation_9]
  properties {soundClassOverride=&Fort_WeaponRanged_Fire_3P_PSM}
 }
 {id=WavePlayer_1, properties {soundWaveAssetPtr=/Game/Athena/Sounds/Weapons/Guided_Missile/Unvault/GuidedMissile_Unvault_Launch_01}}
 {id=WavePlayer_4, properties {soundWaveAssetPtr=/Game/Athena/Sounds/Weapons/Guided_Missile/Unvault/GuidedMissile_Unvault_Launch_02}}
 {id=WavePlayer_5, properties {soundWaveAssetPtr=/Game/Athena/Sounds/Weapons/Guided_Missile/Unvault/GuidedMissile_Unvault_Launch_03}}
]
properties {
 maxDistance=4950
 soundClassObject=&Fort_WeaponRanged_Fire_3P_PSM
 volumeMultiplier=1
 concurrencyOverrides {bLimitToOwner=true, maxCount=1, resolutionRule=PreventNew}
 concurrencySet [/Game/Athena/Sounds/Weapons/Guided_Missile/Unvault/GuidedMissile_Unvault_Fire]
}
