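# Sound cue for the end of the Guided Missile reload, classed as weapon foley with weapon-reload attenuation.
# A mixer layers two reload waves, ReloadC and ReloadB, each through its own modulator.
# The ReloadC layer starts immediately, while ReloadB follows after a short fixed delay, pitched down and mixed louder, giving a staggered two-part sound.
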SoundCue /Game/Athena/Sounds/Weapons/Guided_Missile/GuidedMissile_Reload_End_Cue
root=Mixer_0
nodes [
 {id=Delay_0, children [WavePlayer_1], properties {delayMax=0.35, delayMin=0.35}}
 {id=Mixer_0, children [Modulator_1,Modulator_0], properties {inputVolume [1,1.3]}}
 {id=Modulator_0, children [Delay_0], properties {pitchMax=0.9, pitchMin=0.9}}
 {id=Modulator_1, children [WavePlayer_0]}
 {id=WavePlayer_0, properties {soundWaveAssetPtr=&/GuidedMissile_ReloadC_01}}
 {id=WavePlayer_1, properties {soundWaveAssetPtr=&/GuidedMissile_ReloadB_01}}
]
properties {
 attenuationSettings=&Fort_Weapon_Reload_Att
 maxDistance=2250
 soundClassObject=&Fort_WeaponFoley
 volumeMultiplier=0.752
}
