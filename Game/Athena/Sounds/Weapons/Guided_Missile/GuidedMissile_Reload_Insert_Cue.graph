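# Sound cue for the reload insert foley of the guided missile weapon.
# It has no root or sound nodes, so it references no audio as authored.
# It still carries the shared weapon reload attenuation and the weapon foley sound class.
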SoundCue /Game/Athena/Sounds/Weapons/Guided_Missile/GuidedMissile_Reload_Insert_Cue
root=null
nodes []
properties {
 attenuationSettings=&Fort_Weapon_Reload_Att
 maxDistance=5500
 soundClassObject=&Fort_WeaponFoley
 volumeMultiplier=1
}
