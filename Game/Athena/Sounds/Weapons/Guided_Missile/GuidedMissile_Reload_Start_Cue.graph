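# Sound cue for the start of the Guided Missile reload in Athena.
# A single wave player plays the GuidedMissile_ReloadA_01 sound through a modulator that randomly raises its pitch slightly on each play.
# It uses the weapon foley sound class and the shared weapon reload attenuation, with slightly reduced volume.
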SoundCue /Game/Athena/Sounds/Weapons/Guided_Missile/GuidedMissile_Reload_Start_Cue
root=Modulator_1
nodes [
 {id=Modulator_1, children [WavePlayer_0], properties {pitchMax=1.2, pitchMin=1.1}}
 {id=WavePlayer_0, properties {soundWaveAssetPtr=&/GuidedMissile_ReloadA_01}}
]
properties {
 attenuationSettings=&Fort_Weapon_Reload_Att
 maxDistance=2250
 soundClassObject=&Fort_WeaponFoley
 volumeMultiplier=0.8
}
