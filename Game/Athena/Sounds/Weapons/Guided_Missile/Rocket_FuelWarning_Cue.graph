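# Looping sound cue for the guided missile's low-fuel warning, routed to the weapons sound class.
# A mixer layers two looping warning waves, with the B layer quieter and slightly pitched down by a fixed modulator.
# Attenuation overrides turn off distance falloff and spatialization, so it plays as a flat 2D sound.
# Concurrency overrides allow only one instance and block new ones while it plays.
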SoundCue /Game/Athena/Sounds/Weapons/Guided_Missile/Rocket_FuelWarning_Cue
root=Mixer_0
nodes [
 {id=Mixer_0, children [WavePlayer_6,Modulator_0], properties {inputVolume [1,0.68]}}
 {id=Modulator_0, children [WavePlayer_0], properties {pitchMax=0.912, pitchMin=0.912, volumeMax=1, volumeMin=1}}
 {
  id=WavePlayer_0
  properties {bLooping=true, soundWaveAssetPtr=/Game/Athena/Sounds/Weapons/Guided_Missile/Unvault/GuidedMissile_Unvault_FuelWarning_B_01}
 }
 {
  id=WavePlayer_6
  properties {bLooping=true, soundWaveAssetPtr=/Game/Athena/Sounds/Weapons/Guided_Missile/Unvault/GuidedMissile_Unvault_FuelWarning_A_01}
 }
]
properties {
 attenuationSettings=/Game/Sounds/Attenuation/Fort_Jetpack_Att
 bOverrideAttenuation=true
 bOverrideConcurrency=true
 maxDistance=2097152
 soundClassObject=/Game/Sounds/FortSoundClassesAndModes/Fort_Weapons
 volumeMultiplier=0.664
 attenuationOverrides {bAttenuate=false, bSpatialize=false, falloffDistance=1400, occlusionTraceChannel=ECC_GameTraceChannel8, attenuationShapeExtents {x=600}}
 concurrencyOverrides {maxCount=1, resolutionRule=PreventNew}
}
